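# Niagara trail effect for the Chicken Warrior pickaxe, drawing two white tendril ribbons between bone or socket pairs.
# Both short-lived AnimTrail ribbons drift upward through panning curl noise and positive-Z gravity, with width tapering over a curve.
# FN_RibbonMaterial pans textures and fades color alpha over emitter loop age, while the second ribbon also erodes over particle age.
# A disabled TurkeyDrippings emitter would burst occasional falling meshes from the user skeletal mesh.
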NiagaraSystem /BRCosmetics/Weapons/FORT_Melee/Pickaxe_Chicken_Warrior_Male_1h/FX/NS_Pickaxe_ChickenWarrior_Trail
dependencies {
 usedDataInterfaces [&NiagaraDataInterfaceCurve, &NiagaraDataInterfaceColorCurve, &NiagaraDataInterfaceVectorCurve]
 usedDynamicInputs []
 usedModules [&SystemState]
}
emitters {
 Tendril_Trail {
  emitterUpdateScript [
   &EmitterState(Life Cycle Mode=::NewEnumerator1, Loop Behavior=::NewEnumerator1, Loop Delay=0.1, Loop Duration=1, UseLoopDelay=true)
   &FN_SpawnAnimTrail_v2()
  ]
  particleSpawnScript [
   /Game/Effects/Niagara/Modules/AnimTrail/FN_InitializeAnimTrail/FN_InitializeAnimTrailBoneIndecies(Second Bone or Socket Index=1)
   /Game/Effects/Niagara/Modules/AnimTrail/FN_InitializeAnimTrail/FN_InitializeAnimTrail_v2(Lifetime=0.3)
  ]
  particleUpdateScript [
   &FN_UpdateAnimTrail_v2()
   /SetVariables(Particles.RibbonFacing=<1 0 0>) {enabled=false, isSetParameters=true, module=SetVariables_#1}
   /Niagara/Modules/Update/Forces/CurlNoiseForce(Noise Frequency=50, Noise Strength=300, Pan Noise=true, Pan Noise Field=<0.1 0.1 0.1>)
   &GravityForce(Gravity=<0 0 350>)
   &SolveForcesAndVelocity(Clamp Velocity=false, Limit Acceleration=false)
   &ScaleRibbonWidth(Ribbon Width Scale=&FloatFromCurve(FloatCurve=DataInterface() {propertyValues {Curve {keys [<at 0 1 RCIM_Cubic>, {interpMode=RCIM_Cubic, leaveTangent=-2.0105, time=1, value=0}]}}}))
   &FN_RibbonMaterial(
    Color=&ColorFromCurve(
     ColorCurve=DataInterface() {propertyValues {
       AlphaCurve {keys [<at 0.0221 0>,<at 0.3206 0.05 RCIM_Cubic>,<at 1 0 RCIM_Cubic>]}
       BlueCurve {keys [<at 0.0202 1 RCIM_Cubic>]}
       GreenCurve {keys [<at 0.0202 1 RCIM_Cubic>]}
       RedCurve {keys [<at 0.0202 1 RCIM_Cubic>]}
      }}
     CurveIndex=$Emitter.NormalizedLoopAge
    )
    Erosion Amount:NiagaraFloat=&FloatFromCurve(
     FloatCurve=DataInterface() {propertyValues {Curve {keys [{arriveTangent=5.1387, interpMode=RCIM_Cubic, leaveTangent=5.1387, tangentMode=RCTM_User, time=0, value=0}, <at 1 1.5 RCIM_Cubic>]}}}
     Scale Curve=0.5
    )
    Erosion Pan Speed U=1
    Opacity Pan U Speed=0.25
   ) {enabled=false, module=FN_RibbonMaterial001}
   &FN_RibbonMaterial(
    Color=&ColorFromCurve(
     ColorCurve=DataInterface() {propertyValues {
       AlphaCurve {keys [<at 0.0033 0>,<at 0.1817 0.2 RCIM_Cubic>,<at 1 0 RCIM_Cubic>]}
       BlueCurve {keys [<at 0.0202 1 RCIM_Cubic>]}
       GreenCurve {keys [<at 0.0202 1 RCIM_Cubic>]}
       RedCurve {keys [<at 0.0202 1 RCIM_Cubic>]}
      }}
     CurveIndex=$Emitter.NormalizedLoopAge
    )
    Erosion Pan Speed U=1
    Opacity Pan U Speed=0.25
    Scale Alpha:NiagaraFloat=&FloatFromCurve(
     FloatCurve=DataInterface() {propertyValues {Curve {keys [<at 0 0 RCIM_Cubic>,<at 0.3219 1 RCIM_Cubic>,<at 1 0 RCIM_Cubic>]}}}
     Scale Curve=0.2
    )
   )
  ]
  rendererClasses [&NiagaraRibbonRendererProperties]
  renderers [{rendererClass=&NiagaraRibbonRendererProperties}]
 }
 Tendril_Trail001 {
  emitterUpdateScript [
   &EmitterState(Life Cycle Mode=::NewEnumerator1, Loop Behavior=::NewEnumerator1, Loop Delay=0.1, Loop Duration=1, UseLoopDelay=true)
   &FN_SpawnAnimTrail_v2()
  ]
  particleSpawnScript [
   &FN_InitializeAnimTrailBoneIndecies(First Bone or Socket Index=2, Second Bone or Socket Index=3)
   &FN_InitializeAnimTrail_v2(AnimTrailWidth=3, Lifetime=0.3)
  ]
  particleUpdateScript [
   &FN_UpdateAnimTrail_v2()
   /SetVariables(Particles.RibbonFacing=<1 0 0>) {enabled=false, isSetParameters=true, module=SetVariables_#1}
   &CurlNoiseForce(Noise Frequency=40, Noise Strength=300, Pan Noise=true, Pan Noise Field=<0.1 0.1 0.5>)
   &GravityForce(Gravity=<0 0 250>)
   &SolveForcesAndVelocity(Clamp Velocity=false, Limit Acceleration=false)
   &ScaleRibbonWidth(Ribbon Width Scale=&FloatFromCurve(FloatCurve=DataInterface() {propertyValues {Curve {keys [<at 0 1 RCIM_Cubic>, {interpMode=RCIM_Cubic, leaveTangent=-2.0105, time=1, value=0}]}}}))
   &FN_RibbonMaterial(
    Color=&ColorFromCurve(
     ColorCurve=DataInterface() {propertyValues {
       AlphaCurve {keys [<at 0.0015 0>,<at 0.1862 1 RCIM_Cubic>,<at 1 0 RCIM_Cubic>]}
       BlueCurve {keys [<at 0.0202 1 RCIM_Cubic>]}
       GreenCurve {keys [<at 0.0202 1 RCIM_Cubic>]}
       RedCurve {keys [<at 0.0202 1 RCIM_Cubic>]}
      }}
     CurveIndex=$Emitter.NormalizedLoopAge
    )
    Color Pan U Speed=1
    Color Scale U=2
    Erosion Amount:NiagaraFloat=&Lerp_Float(A=0, B=1.75, Alpha=$Particles.NormalizedAge)
    Erosion Pan Speed U=1
    Erosion Power=0.25
    Erossion Scale U=2
    Opacity Pan U Speed=0.23
    Opacity Scale U=0.2
    Scale Alpha:NiagaraFloat=&FloatFromCurve(
     FloatCurve=DataInterface() {propertyValues {Curve {keys [<at 0 0 RCIM_Cubic>,<at 0.3219 1 RCIM_Cubic>,<at 1 0 RCIM_Cubic>]}}}
     Scale Curve=0.4
    )
    TIme:NiagaraFloat=&Divide_Float(A=$Engine.Time, B=0.5)
   )
  ]
  rendererClasses [&NiagaraRibbonRendererProperties]
  renderers [{rendererClass=&NiagaraRibbonRendererProperties}]
 }
 TurkeyDrippings {
  bEnabled=false
  emitterUpdateScript [
   &EmitterState(
    Life Cycle Mode=::NewEnumerator1
    Loop Duration=&RandomRangeFloat(Minimum=0, Maximum=4)
    Recalculate Duration Each Loop=true
    UseLoopDelay=false
   )
   /SetVariables(Emitter.PlayerVelocityVector:Vector3f=$Engine.Owner.Velocity) {isSetParameters=true}
   /SetVariables(Emitter.PlayerSpeed:NiagaraFloat=&VectorLength(Vector=$Emitter.PlayerVelocityVector)) {isSetParameters=true}
   &SpawnBurst_Instantaneous(Spawn Count=1, Use Spawn Probability=true)
   &SpawnRate(Spawn Probability=0.2, Use Spawn Probability=true) {enabled=false}
  ]
  particleSpawnScript [
   /Niagara/Modules/Spawn/Initialization/InitializeParticle(
    Color=/Niagara/DynamicInputs/TypeConversions/MakeLinearColorFromFloat(Value:NiagaraFloat=&RandomRangeFloat(Minimum=0.3, Maximum=0.333))
    Lifetime=1.1
    Mesh Scale=&Multiply_VectorByFloat(Vector=<1 1 1>, Float=&RandomRangeFloat(Minimum=0.8, Maximum=1.2))
    Module.Write Color=true
    Module.Write Lifetime=true
    Module.Write Mass=true
    Module.Write Material Random=true
    Module.Write Position=false
    Module.Write Scale=true
    Module.Write SpriteRotation=false
    Module.Write SpriteSize=true
    Position=&SimulationPosition()
    Sprite Size=<5 5>
   )
   &SkeletalMeshLocation(
    Inherit Velocity=false
    MaskSpawnGroup=false
    Mesh Sampling Type=::NewEnumerator5
    Module.Override Randomness=false
    Sampled Position Offset=&RandomRangeVector(Minimum=<-10 -10 0>, Maximum=<10 10 0>)
    Skeletal Mesh=$User.SkeletalMesh
   )
   &ApplyInitialForces()
   &InheritVelocity(Module.Limit Speed=true, Velocity Scale=&Multiply_VectorByFloat(Vector=<1 1 1>, Float=-0.1)) {module=InheritVelocity001}
  ]
  particleUpdateScript [
   &ParticleState()
   &GravityForce(Gravity=&Multiply_VectorByFloat(
     Vector=<1 1 -700>
     Float=&FloatFromCurve(FloatCurve=DataInterface() {propertyValues {Curve {keys [{arriveTangent=1.8039, interpMode=RCIM_Cubic, leaveTangent=1.8039, tangentMode=RCTM_User, time=0, value=0}, <at 1 1>]}}})
    ))
   &SolveForcesAndVelocity(Clamp Velocity=false, Limit Acceleration=false)
   &ScaleMeshSize(Scale Factor=&Multiply_Vector(
     A=<1 1 1>
     B=&VectorFromCurve(VectorCurve=DataInterface() {propertyValues {
        XCurve {keys [
          <at 0 0.5 RCIM_Cubic>
          {arriveTangent=0.5, interpMode=RCIM_Cubic, leaveTangent=-5.3087, tangentMode=RCTM_Break, time=0.1222, value=2.3}
          <at 1 0.5>
         ]}
        YCurve {keys [<at 0 1>,<at 1 1>]}
        ZCurve {keys [<at 0 1>,<at 1 1>]}
       }})
    ))
   &DynamicMaterialParameters(
    Index 0 Param 1=0.8
    Index 0 Param 2=&FloatFromCurve(FloatCurve=DataInterface() {propertyValues {Curve {keys [{arriveTangent=-1.8884, interpMode=RCIM_Cubic, leaveTangent=-1.8884, tangentMode=RCTM_User, time=0, value=1}, <at 1 0.26>]}}})
    Index 0 Param 3=1
    Index 0 Param 4=1
    Module.Param0WriteEnabled=true
    Module.Param1WriteEnabled=true
    Module.Param2WriteEnabled=true
    Module.Param3WriteEnabled=true
   )
   &ScaleColor(Scale Alpha=&FN_FrontEndFloatModifier(Gameplay=0.1, Front End=0), ScaleA=true, ScaleRGB=true) {enabled=false}
  ]
  rendererClasses [&NiagaraMeshRendererProperties]
  renderers [{rendererClass=&NiagaraMeshRendererProperties, rendererIndex=0}]
 }
}
summary {
 systemName=NS_Pickaxe_ChickenWarrior_Trail
 emitters [
  {emitterName=Tendril_Trail, rendererClasses [&NiagaraRibbonRendererProperties]}
  {emitterName=Tendril_Trail001, rendererClasses [&NiagaraRibbonRendererProperties]}
  {bEnabled=false, emitterName=TurkeyDrippings, rendererClasses [&NiagaraMeshRendererProperties]}
 ]
 userVariables [
  User.RibbonIDOffset {type=NiagaraInt32}
  User.SkeletalMesh {
   type=NiagaraDataInterfaceSkeletalMesh
   defaultValue {
    struct=NiagaraExt_VariableValue_DataInterface
    value {dataInterface=&self:NiagaraDataInterfaceSkeletalMesh_1}
   }
  }
 ]
}
