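# Idle Niagara effect for the chicken pickaxe, adding steam-like smoke, drippings and trails, all positioned from a user-supplied skeletal mesh.
# SmokeStationary sprites rise from the mesh, fading with distance from their source until killed.
# TurkeyDrippings mesh particles spawn at random intervals and with movement, inherit velocity, then fall and stretch.
# FN_AttachedRibbon draws two faint, eroding, upward-drifting ribbons that partly follow source movement and cull by distance and visibility.
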NiagaraSystem /BRCosmetics/Weapons/FORT_Melee/Pickaxe_Chicken_Warrior_Male_1h/FX/NS_Pickaxe_Chicken_Idle
dependencies {
 usedDataInterfaces [&NiagaraDataInterfaceCurve, &NiagaraDataInterfaceVectorCurve, &NiagaraDataInterfaceColorCurve]
 usedDynamicInputs [&NormalizeFloat]
 usedModules [
  &SystemState
  &self:SystemSpawnScript.SystemScriptSource.SystemScriptGraph.NiagaraNodeAssignment_0.SetVariables_#1
 ]
}
emitters {
 FN_AttachedRibbon {
  emitterUpdateScript [
   &EmitterState(
    Enable Distance Culling=true
    Enable Visibility Culling=true
    Max Distance Response=::SleepAndClearParticles
    MaxDistance=1000
    Module.bUseMaxDistance=true
    Module.bUseMinDistance=false
    Scalability Mode=::NewEnumerator1
    Visibility Cull Delay=0
    Visibility Cull Response:ENiagaraExecutionStateManagement=::SleepAndClearParticles
   )
   &FN_SpawnRibbon_v2(Number of Ribbons=2, SpawnRate=30)
   &FN_Integrator() {enabled=false}
  ]
  particleSpawnScript [
   /Game/Effects/Niagara/Modules/Ribbons/FN_InitializeRibbon_v4(Enable Skeletal Position=true, Lifetime=1, Ribbon Width=20, SkeletalMesh:NiagaraDataInterfaceSkeletalMesh=$User.SkeletalMesh)
   /SetVariables(Particles.RandomFromRibbonID:NiagaraFloat=/Game/Effects/Niagara/DynamicInputs/Random/FN_SeededRandomRangeFloat(Minimum:NiagaraFloat=0, Maximum:NiagaraFloat=1, Random Seed:NiagaraInt32=$Particles.SpawnGroup)) {isSetParameters=true}
   &AddVelocity(Coordinate Space=::World, Velocity=<0 0 50>)
   &ApplyInitialForces()
  ]
  particleUpdateScript [
   /SetVariables(Particles.Lifetime=&FloatFromCurve(FloatCurve=DataInterface() {propertyValues {Curve {keys [<at 0 1>,<at 1 0.2>]}}}, CurveIndex=$System.FN_Integrator.Value, Scale Curve=1.5)) {isSetParameters=true}
   &InheritSourceMovement(
    Applied Position Scale=&VectorFromFloat(Value=&FloatFromCurve(FloatCurve=DataInterface() {propertyValues {Curve {keys [{arriveTangent=-3.6922, interpMode=RCIM_Cubic, leaveTangent=-3.6922, tangentMode=RCTM_User, time=0, value=0.8}, <at 0.7 0 RCIM_Cubic>]}}}))
    ApplyForce=false
    ApplyPosition=true
    ApplyPositionFalloff=false
   )
   &SolveForcesAndVelocity(Clamp Velocity=false, Limit Acceleration=false)
   /Game/Effects/Niagara/Modules/Ribbons/FN_UpdateRibbon_v4(Enable Skeletal Position=true, SkeletalMesh:NiagaraDataInterfaceSkeletalMesh=$User.SkeletalMesh)
   &FN_RibbonMaterial(
    Color=&ColorFromCurve(ColorCurve=DataInterface() {propertyValues {
       AlphaCurve {keys [<at 0 0.04 RCIM_Cubic>,<at 1 0 RCIM_Cubic>]}
       BlueCurve {keys [<at 0 1 RCIM_Cubic>]}
       GreenCurve {keys [<at 0 1 RCIM_Cubic>]}
       RedCurve {keys [<at 0 1 RCIM_Cubic>]}
      }})
    Erosion Amount:NiagaraFloat=&FloatFromCurve(
     FloatCurve=DataInterface() {propertyValues {Curve {keys [{arriveTangent=5.1387, interpMode=RCIM_Cubic, leaveTangent=5.1387, tangentMode=RCTM_User, time=0, value=0}, <at 1 1.5 RCIM_Cubic>]}}}
     Scale Curve=0.5
    )
    Erosion Pan Speed U=1
    Opacity Pan U Offset:NiagaraFloat=$Particles.RandomFromRibbonID
    Opacity Pan U Speed=0.5
   ) {module=FN_RibbonMaterial001}
  ]
  rendererClasses [&NiagaraRibbonRendererProperties]
  renderers [{rendererClass=&NiagaraRibbonRendererProperties}]
 }
 SmokeStationary {
  emitterUpdateScript [
   &EmitterState()
   &SpawnRate(SpawnRate=15, Use Spawn Probability=false)
  ]
  particleSpawnScript [
   /SetVariables(Particles.SpawnGroup=&SpawnGroup()) {isSetParameters=true}
   /Niagara/Modules/Spawn/Initialization/InitializeParticle(
    Color=<rgba 1 1 1 0.05>
    Lifetime=&RandomRangeFloat(Minimum=0.8, Maximum=1.2)
    Module.Write Color=true
    Module.Write Lifetime=true
    Module.Write Mass=false
    Module.Write Material Random=true
    Module.Write Position=false
    Module.Write Scale=false
    Module.Write SpriteRotation=true
    Module.Write SpriteSize=true
    Position=&SimulationPosition()
    Sprite Rotation=&UniformRangedFloat(Minimum=0, Maximum=360)
    Sprite Size=&Vector2DFromFloat(Value=&UniformRangedFloat(Minimum=15, Maximum=25))
   )
   &SphereLocation(
    MaskSpawnGroup=false
    Override Seed=false
    Sphere Radius=12
    Surface Only=false
    UseNonUniformScale=false
    UseOffset=false
    UseSphereOrientation=false
   )
   &SkeletalMeshLocation(
    Inherit Velocity=false
    MaskSpawnGroup=true
    Mesh Sampling Type=::NewEnumerator5
    Sampled Position Offset=&Multiply_Vector(A=<1 1 1>, B=$Output.SphereLocation.SphereVector)
    Sampled Position Offset Coordinate Space:ENiagaraMeshTransforms=::NewEnumerator3
    Skeletal Mesh=$User.SkeletalMesh
    Socket Sampling Mode=::NewEnumerator5
   )
   &AddVelocity(Velocity=&RandomRangeVector(Minimum=<0 0 2>, Maximum=<1 1 8>))
   &CameraOffset(Camera Offset Amount=5)
   &InheritVelocity(Module.Limit Speed=true, Velocity Limit=20, Velocity Scale=&VectorFromFloat(Value=0.25))
  ]
  particleUpdateScript [
   &ParticleState()
   &SkeletalMeshLocation(
    Apply To Particle Position=false
    Inherit Velocity=false
    MaskSpawnGroup=false
    Mesh Sampling Type=::NewEnumerator5
    Skeletal Mesh=$User.SkeletalMesh
    Socket Sampling Mode=::NewEnumerator5
   ) {module=SkeletalMeshLocation001}
   /SetVariables(Particles.DistanceFromSource:NiagaraFloat=&ClampFloat(Float=&Divide_Float(
      A=&Add_Float(
       A=0
       B=&VectorLength(Vector=&Subtract_Vector(A=Linked() {linkedVariable {name=Particles.Position, type=NiagaraPosition}}, B=$Output.SkeletalMeshLocation001.SampledPosition))
      )
      B=20
     ))) {isSetParameters=true}
   &GravityForce(Gravity=<0 0 20>)
   &SolveForcesAndVelocity(Clamp Velocity=false, Limit Acceleration=false)
   &ScaleSpriteSize(Scale Factor=&Vector2DFromFloat(Value=&FloatFromCurve(FloatCurve=DataInterface() {propertyValues {Curve {keys [{arriveTangent=-0.28, interpMode=RCIM_Cubic, leaveTangent=-0.28, tangentMode=RCTM_User, time=0, value=0.5}, <at 1 1>]}}}))) {module=ScaleSpriteSize001}
   &ScaleSpriteSize(Scale Factor=&Vector2DFromFloat(Value=&FloatFromCurve(FloatCurve=DataInterface() {propertyValues {Curve {keys [
          {arriveTangent=-0.0248, interpMode=RCIM_Cubic, leaveTangent=-0.0248, tangentMode=RCTM_User, time=0, value=1}
          {arriveTangent=-0.5458, interpMode=RCIM_Cubic, leaveTangent=-0.5458, tangentMode=RCTM_User, time=1, value=0.75}
         ]}}})))
   &DynamicMaterialParameters(
    Index 0 Param 1=&FloatFromCurve(
     FloatCurve=DataInterface() {propertyValues {Curve {keys [
         {arriveTangent=4.0913, interpMode=RCIM_Cubic, leaveTangent=4.0913, tangentMode=RCTM_User, time=0, value=0}
         {arriveTangent=0.1513, interpMode=RCIM_Cubic, leaveTangent=0.1513, tangentMode=RCTM_User, time=0.5, value=1}
         {arriveTangent=-0.0082, interpMode=RCIM_Cubic, leaveTangent=-0.0082, tangentMode=RCTM_User, time=1, value=0}
        ]}}}
     Scale Curve=0.6
    )
    Index 0 Param 4=1
    Module.Param0WriteEnabled=true
    Module.Param1WriteEnabled=false
    Module.Param2WriteEnabled=false
    Module.Param3WriteEnabled=true
   )
   /SetVariables(Particles.StartFrame:NiagaraFloat=&RandomRangeFloat(Minimum=0, Maximum=63)) {isSetParameters=true}
   /SetVariables(Particles.SubImageIndex=&Lerp_Float(
     A=$Particles.StartFrame
     B=&Add_Float(A=$Particles.StartFrame, B=&RandomRangeFloat(Minimum=2, Maximum=6))
     Alpha=$Particles.NormalizedAge
    )) {isSetParameters=true}
   &Color(Color=&LerpLinearColors(StartColor=<rgba 1 0 0 1>, EndColor=<rgba 0 0 1 1>, LerpFactor=$Particles.DistanceFromSource)) {enabled=false}
   &ScaleColor(Scale Alpha=&Lerp_Float(A=1, B=0, Alpha=$Particles.DistanceFromSource), ScaleA=true, ScaleRGB=true)
   &KillParticles(Kill Particles=&SetBoolByFloatComparison(
     A=&MakeFloatFromLinearColor(LinearColor=$Particles.Color, Channel:ENiagaraLinearColor_Channels=::NewEnumerator3)
     B=0
     Comparison Type=::NewEnumerator4
    ))
  ]
  rendererClasses [&NiagaraSpriteRendererProperties]
  renderers [{rendererClass=&NiagaraSpriteRendererProperties}]
 }
 TurkeyDrippings {
  emitterUpdateScript [
   &EmitterState(
    Life Cycle Mode=::NewEnumerator1
    Loop Duration=&RandomRangeFloat(Minimum=0, Maximum=4)
    Recalculate Duration Each Loop=true
    UseLoopDelay=false
   )
   /SetVariables(Emitter.PlayerVelocityVector:Vector3f=$Engine.Owner.Velocity) {enabled=false, isSetParameters=true}
   /SetVariables(Emitter.PlayerSpeed:NiagaraFloat=&VectorLength(Vector=$Emitter.PlayerVelocityVector)) {enabled=false, isSetParameters=true}
   &SpawnBurst_Instantaneous(Spawn Count=1, Use Spawn Probability=true)
   &SpawnRate(Spawn Probability=0.2, SpawnRate=100, Use Spawn Probability=true) {enabled=false}
   &SpawnPerUnit(Module.Use Max Movement Threshold=true, Module.Use Movement Tolerance=false, Spawn Per Unit=100, Use Spawn Probability=false)
  ]
  particleSpawnScript [
   &InitializeParticle(
    Color=/Niagara/DynamicInputs/TypeConversions/MakeLinearColorFromFloat(Value:NiagaraFloat=&RandomRangeFloat(Minimum=0.3, Maximum=0.333))
    Lifetime=1.1
    Mesh Scale=&Multiply_VectorByFloat(Vector=<1 1 1>, Float=&RandomRangeFloat(Minimum=0.8, Maximum=1.2))
    Module.Write Color=true
    Module.Write Lifetime=true
    Module.Write Mass=true
    Module.Write Material Random=true
    Module.Write Position=false
    Module.Write Scale=true
    Module.Write SpriteRotation=false
    Module.Write SpriteSize=true
    Position=&SimulationPosition()
    Sprite Size=<5 5>
   )
   &SkeletalMeshLocation(
    Inherit Velocity=false
    MaskSpawnGroup=false
    Mesh Sampling Type=::NewEnumerator5
    Sampled Position Offset=&RandomRangeVector(Minimum=<-5 -5 -10>, Maximum=<5 5 5>)
    Skeletal Mesh=$User.SkeletalMesh
    Socket Sampling Mode=::NewEnumerator5
   )
   &InheritVelocity(Module.Limit Speed=true, Velocity Limit=150, Velocity Scale=&Multiply_VectorByFloat(Vector=<1 1 1>, Float=1)) {module=InheritVelocity001}
   &ApplyInitialForces() {enabled=false}
  ]
  particleUpdateScript [
   &ParticleState()
   &InheritSourceMovement(
    Applied Position Scale=&VectorFromFloat(Value=&FloatFromCurve(FloatCurve=DataInterface() {propertyValues {Curve {keys [{arriveTangent=-1.5769, interpMode=RCIM_Cubic, leaveTangent=-1.5769, tangentMode=RCTM_User, time=0, value=0.3614}, <at 1 0 RCIM_Cubic>]}}}))
    ApplyForce=false
    ApplyPosition=true
    ApplyPositionFalloff=false
   )
   &GravityForce(Gravity=<0 0 -400>)
   &SolveForcesAndVelocity(Clamp Velocity=false, Limit Acceleration=false)
   &CalculateAccurateVelocity() {enabled=false}
   &ScaleMeshSize(Scale Factor=&Multiply_Vector(
     A=<1 1 1>
     B=&VectorFromCurve(VectorCurve=DataInterface() {propertyValues {
        XCurve {keys [
          <at 0 0.5 RCIM_Cubic>
          {arriveTangent=0.5, interpMode=RCIM_Cubic, leaveTangent=-5.3087, tangentMode=RCTM_Break, time=0.1222, value=2.3}
          <at 1 0.5>
         ]}
        YCurve {keys [<at 0 1>,<at 1 1>]}
        ZCurve {keys [<at 0 1>,<at 1 1>]}
       }})
    ))
   &DynamicMaterialParameters(
    Index 0 Param 1=0.8
    Index 0 Param 2=&FloatFromCurve(FloatCurve=DataInterface() {propertyValues {Curve {keys [{arriveTangent=-1.8884, interpMode=RCIM_Cubic, leaveTangent=-1.8884, tangentMode=RCTM_User, time=0, value=1}, <at 1 0.26>]}}})
    Index 0 Param 3=1
    Index 0 Param 4=1
    Module.Param0WriteEnabled=true
    Module.Param1WriteEnabled=true
    Module.Param2WriteEnabled=true
    Module.Param3WriteEnabled=true
   )
   &ScaleColor(Scale Alpha=&FN_FrontEndFloatModifier(Gameplay=0.1, Front End=0), ScaleA=true, ScaleRGB=true) {enabled=false}
  ]
  rendererClasses [&NiagaraMeshRendererProperties]
  renderers [{rendererClass=&NiagaraMeshRendererProperties, rendererIndex=0}]
 }
}
summary {
 systemName=NS_Pickaxe_Chicken_Idle
 emitters [
  {emitterName=SmokeStationary, rendererClasses [&NiagaraSpriteRendererProperties]}
  {emitterName=TurkeyDrippings, rendererClasses [&NiagaraMeshRendererProperties]}
  {emitterName=FN_AttachedRibbon, rendererClasses [&NiagaraRibbonRendererProperties]}
 ]
 userVariables [
  User.bIsFrontendPreview {type=NiagaraBool}
  User.IsSkydiving {type=NiagaraInt32}
  User.RibbonIDOffset {type=NiagaraInt32}
  User.SkeletalMesh {
   type=NiagaraDataInterfaceSkeletalMesh
   defaultValue {
    struct=NiagaraExt_VariableValue_DataInterface
    value {dataInterface=&self:NiagaraDataInterfaceSkeletalMesh_0}
   }
  }
 ]
}
