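# Physics asset for the one-handed male Chicken Warrior pickaxe cosmetic, defining secondary-motion bodies for its dangling flaps.
# A single capsule on the Melee_TwoHanded_Handle bone anchors capsule bodies on the dyn_flap bones along both sides and the handle.
# First flap segments constrain to the handle and second segments to their first, forming short chains.
# Every joint shares the same tight swing and twist limits, keeping movement small.
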PhysicsAsset /BRCosmetics/Weapons/FORT_Melee/Pickaxe_Chicken_Warrior_Male_1h/Meshes/Chicken_Warrior_Male_1h_Physics
bodies {
 Melee_TwoHanded_Handle [{shapeType=Capsule, length=38.3465, radius=1.9403, center=<-0.6528 -0.1372 14.8944>, rotation=<rot -0.2393 -90 2.2137>}]
 dyn_flapA_L_01 [{shapeType=Capsule, length=3.4792, radius=0.8006, center=<2 0 0>, rotation=<rot 90 0 0>}]
 dyn_flapA_R_01 [{shapeType=Capsule, length=3.4792, radius=0.8056, center=<2 0 0>, rotation=<rot 90 0 0>}]
 dyn_flapA_R_02 [{shapeType=Capsule, length=1.6845, radius=2.0935, center=<2 0 0>, rotation=<rot 90 0 0>}]
 dyn_flapB_L_01 [{shapeType=Capsule, length=3.4792, radius=1.7652, center=<2 0 0>, rotation=<rot 90 0 0>}]
 dyn_flapB_L_02 [{shapeType=Capsule, length=1.6845, radius=0.505, center=<2 0 0>, rotation=<rot 90 0 0>}]
 dyn_flapB_R_01 [{shapeType=Capsule, length=3.4792, radius=0.8056, center=<2 0 0>, rotation=<rot 90 0 0>}]
 dyn_flapC_L_01 [{shapeType=Capsule, length=3.4792, radius=0.8056, center=<2 0 0>, rotation=<rot 90 0 0>}]
 dyn_flapC_L_02 [{shapeType=Capsule, length=1.6845, radius=2.0935, center=<2 0 0>, rotation=<rot 90 0 0>}]
 dyn_flapC_R_01 [{shapeType=Capsule, length=3.4792, radius=0.8056, center=<2 0 0>, rotation=<rot 90 0 0>}]
 dyn_flapD_R_01 [{shapeType=Capsule, length=3.4792, radius=0.8056, center=<2 0 0>, rotation=<rot 90 0 0>}]
 dyn_flapHandle_01 [{shapeType=Capsule, length=3.4792, radius=1.4874, center=<2 0 0>, rotation=<rot 90 0 0>}]
 dyn_flapHandle_02 [{shapeType=Capsule, length=1.6845, radius=1.3768, center=<2 0 0>, rotation=<rot 90 0 0>}]
}
constraints [
 {bone1Name=dyn_flapC_L_01, bone2Name=Melee_TwoHanded_Handle, swing1LimitDegrees=20, swing2LimitDegrees=10, twistLimitDegrees=4}
 {bone1Name=dyn_flapC_L_02, bone2Name=dyn_flapC_L_01, swing1LimitDegrees=20, swing2LimitDegrees=10, twistLimitDegrees=4}
 {bone1Name=dyn_flapA_L_01, bone2Name=Melee_TwoHanded_Handle, swing1LimitDegrees=20, swing2LimitDegrees=10, twistLimitDegrees=4}
 {bone1Name=dyn_flapB_L_01, bone2Name=Melee_TwoHanded_Handle, swing1LimitDegrees=20, swing2LimitDegrees=10, twistLimitDegrees=4}
 {bone1Name=dyn_flapB_L_02, bone2Name=dyn_flapB_L_01, swing1LimitDegrees=20, swing2LimitDegrees=10, twistLimitDegrees=4}
 {bone1Name=dyn_flapB_R_01, bone2Name=Melee_TwoHanded_Handle, swing1LimitDegrees=20, swing2LimitDegrees=10, twistLimitDegrees=4}
 {bone1Name=dyn_flapC_R_01, bone2Name=Melee_TwoHanded_Handle, swing1LimitDegrees=20, swing2LimitDegrees=10, twistLimitDegrees=4}
 {bone1Name=dyn_flapD_R_01, bone2Name=Melee_TwoHanded_Handle, swing1LimitDegrees=20, swing2LimitDegrees=10, twistLimitDegrees=4}
 {bone1Name=dyn_flapA_R_01, bone2Name=Melee_TwoHanded_Handle, swing1LimitDegrees=20, swing2LimitDegrees=10, twistLimitDegrees=4}
 {bone1Name=dyn_flapHandle_01, bone2Name=Melee_TwoHanded_Handle, swing1LimitDegrees=20, swing2LimitDegrees=10, twistLimitDegrees=4}
 {bone1Name=dyn_flapHandle_02, bone2Name=dyn_flapHandle_01, swing1LimitDegrees=20, swing2LimitDegrees=10, twistLimitDegrees=4}
 {bone1Name=dyn_flapA_R_02, bone2Name=dyn_flapA_R_01, swing1LimitDegrees=20, swing2LimitDegrees=10, twistLimitDegrees=4}
]
