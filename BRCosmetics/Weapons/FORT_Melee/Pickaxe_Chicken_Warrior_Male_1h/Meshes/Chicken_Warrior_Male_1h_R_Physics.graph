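# Physics asset for the right-hand mesh of the one-handed Chicken Warrior Male pickaxe, defining physics for its dangling flaps.
# A capsule body on the handle bone anchors small capsules on the dynamic flap bones.
# Flaps B, C and D hang singly from the handle, while flap A and the handle flap form two-link chains.
# Constraints set tight swing and twist limits, with flap B allowed the widest motion.
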PhysicsAsset /BRCosmetics/Weapons/FORT_Melee/Pickaxe_Chicken_Warrior_Male_1h/Meshes/Chicken_Warrior_Male_1h_R_Physics
bodies {
 Melee_TwoHanded_Handle [{shapeType=Capsule, length=34.9256, radius=4.8082, center=<0.5367 0.0176 15.4876>, rotation=<rot 1.8083 -90 -3.5501>}]
 dyn_flapA_R_01 [{shapeType=Capsule, length=2.1415, radius=0.8922, center=<1.4397 0 0>, rotation=<rot 90 0 0>}]
 dyn_flapA_R_02 [{shapeType=Capsule, length=1.5585, radius=1.0578, center=<1.4728 0 0>, rotation=<rot 90 0 0>}]
 dyn_flapB_R_01 [{shapeType=Capsule, length=4.9081, radius=0.994, center=<1.4397 0 0>, rotation=<rot 90 0 0>}]
 dyn_flapC_R_01 [{shapeType=Capsule, length=2.1415, radius=0.8922, center=<1.4397 0 0>, rotation=<rot 90 0 0>}]
 dyn_flapD_R_01 [{shapeType=Capsule, length=2.9594, radius=1.4849, center=<1.4397 0 0>, rotation=<rot 90 0 0>}]
 dyn_flapHandle_01 [{shapeType=Capsule, length=2.6714, radius=1.3556, center=<1.6013 0 0>, rotation=<rot 90 0 0>}]
 dyn_flapHandle_02 [{shapeType=Capsule, length=1.9827, radius=1.6306, center=<1.6013 0 0>, rotation=<rot 90 0 0>}]
}
constraints [
 {bone1Name=dyn_flapB_R_01, bone2Name=Melee_TwoHanded_Handle, swing1LimitDegrees=45, swing2LimitDegrees=45, twistLimitDegrees=20}
 {bone1Name=dyn_flapC_R_01, bone2Name=Melee_TwoHanded_Handle, swing1LimitDegrees=14, swing2LimitDegrees=20, twistLimitDegrees=4}
 {bone1Name=dyn_flapD_R_01, bone2Name=Melee_TwoHanded_Handle, swing1LimitDegrees=20, swing2LimitDegrees=20, twistLimitDegrees=6}
 {bone1Name=dyn_flapA_R_01, bone2Name=Melee_TwoHanded_Handle, swing1LimitDegrees=20, swing2LimitDegrees=10, twistLimitDegrees=4}
 {bone1Name=dyn_flapA_R_02, bone2Name=dyn_flapA_R_01, swing1LimitDegrees=20, swing2LimitDegrees=10, twistLimitDegrees=4}
 {bone1Name=dyn_flapHandle_01, bone2Name=Melee_TwoHanded_Handle, swing1LimitDegrees=12, swing2LimitDegrees=10, twistLimitDegrees=4}
 {bone1Name=dyn_flapHandle_02, bone2Name=dyn_flapHandle_01, swing1LimitDegrees=12, swing2LimitDegrees=10, twistLimitDegrees=4}
]
